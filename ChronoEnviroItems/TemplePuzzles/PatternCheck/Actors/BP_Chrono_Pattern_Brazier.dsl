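; Interactable temple brazier for the Chrono pattern puzzle that players toggle between lit and unlit.
; Interacting flips the lit state, whose RepNotify swaps fire and smoke effects, plays ignite or extinguish audio, and has the referenced pattern door check its solution.
; A locally viewed pawn hears a 2D sound when interaction begins.
; RepNotify flags also stop startup smoke and disable interaction, the latter triggered by a stop event or the hotfix switch at BeginPlay.
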(event Custom|Stop%20Interactions
  (Variables|Default|SetDisableInteraction true))

(event EventBeginPlay
  ;[1 1 1] Disable braziers if puzzle is disabled
  (if (not (ScalableFloat|GetValueasBool (Variables|Default|GetHotFixEnabled)))
    (Variables|Default|SetDisableInteraction true)))

(event Interaction|EventBlueprintOnInteract
  (params (InteractingPawn :type Fort%20Pawn%20Object%20Reference) (InteractionBeingAttempted :type EInteractionBeingAttempted%20Enum :default "FirstInteraction"))
  (if (Variables|Default|Getb%20Is%20Lit)
    (CallFunction|ExtinguishBrazier)
    (else
      (CallFunction|IgniteBrazier))))

(event Interaction|EventBlueprintOnBeginInteract
  (params (InteractingPawn :type Fort%20Pawn%20Object%20Reference) (InteractionBeingAttempted :type EInteractionBeingAttempted%20Enum :default "FirstInteraction") (InteractionType :type TInteractionType%20Enum :default "IT_NoInteraction"))
  (if (Pawn|IsLocallyViewed InteractingPawn)
    (Audio|PlaySound2D (select (Variables|Default|Getb%20Is%20Lit) (Variables|Default|GetOnInteractExtinguish) (Variables|Default|GetOnInteractLightFlame)))))

(event Custom|Extinguish%20Brazier
  (if (Variables|Default|Getb%20Is%20Lit)
    (Variables|Default|Setb%20Is%20Lit false)))

(event Custom|IgniteBrazier
  (if (not (Variables|Default|Getb%20Is%20Lit))
    (Variables|Default|Setb%20Is%20Lit true)))

(event Custom|StopSmokeOnStartUnlit
  (Variables|Default|SetbStopSmoke true))

(fn BlueprintCanInteract
  (params (InteractingPawn :type Fort%20Pawn%20Object%20Reference) (InteractionBeingAttempted :type EInteractionBeingAttempted%20Enum :default "FirstInteraction") (InteractionType :type TInteractionType%20Enum :default "IT_NoInteraction"))
  (return (Variables|Default|GetCanInteract)))

(fn BlueprintGetFailedInteractionString
  (params (InteractingPawn :type Fort%20Pawn%20Object%20Reference) (InteractionBeingAttempted :type EInteractionBeingAttempted%20Enum :default "FirstInteraction"))
  (return ""))

(fn BlueprintGetInteractionString
  (params (InteractingPawn :type Fort%20Pawn%20Object%20Reference) (InteractionBeingAttempted :type EInteractionBeingAttempted%20Enum :default "FirstInteraction"))
  (return (select (Variables|Default|Getb%20Is%20Lit) (Variables|Default|GetExtinguish%20Interaction%20String) (Variables|Default|GetIgnite%20Interaction%20String))))

(fn GetActorForInteraction
  (return 0))

(fn OnRep_bIsLit
  (bind _firepoof (Variables|Default|GetFirePoof))
  (bind _firepoof_2 (Variables|Default|GetFirePoof)
    (:then
      (Audio|Components|Audio|SetSound _firepoof_2 (select (Variables|Default|Getb%20Is%20Lit) (Variables|Default|GetExtinguishSound) (Variables|Default|GetIgniteSound))))
    (:else
      _))
  (if (Variables|Default|Getb%20Is%20Lit)
    (Audio|Components|Audio|Play _firepoof 0.0)
    ;[1 1 1] Start Fire, Stop Smoke
    (Components|Activation|Activate (Variables|Default|GetAudio_Fire_loop_cue) false)
    (Components|Activation|Activate (Variables|Default|GetNS_Chrono_Temple_Brazier) false)
    (Components|Activation|Deactivate (Variables|Default|GetNS_Chrono_Temple_Brazier_Smoke))
    ;[1 1 1] check if we should open
    (bind _door_ref (Variables|Default|GetDoor%20Ref)
      (:then
        _)
      (:else))
    (else
      (Audio|Components|Audio|Play _firepoof 0.0)
      ;[1 1 1] Douse Fire, Start Smoke
      (Audio|Components|Audio|FadeOut (Variables|Default|GetAudio_Fire_loop_cue) 0.5 0.0 "Sin" :!class /Script/Engine.AudioComponent)
      (Components|Activation|Deactivate (Variables|Default|GetNS_Chrono_Temple_Brazier))
      (Components|Activation|Activate (Variables|Default|GetNS_Chrono_Temple_Brazier_Smoke) false)
      (bind _door_ref_2 (Variables|Default|GetDoor%20Ref)
        (:then
          _)
        (:else))))
  (Class|BPChronoPatternTempleDoor|CheckDoorSolution _door_ref))

(fn OnRep_bStopSmoke
  (Components|Activation|Deactivate (Variables|Default|GetNS_Chrono_Temple_Brazier_Smoke)))

(fn OnRep_DisableInteraction
  (Variables|Interaction|SetbAllowInteract false)
  (Variables|Interaction|SetbShowFirstInteractPrompt false)
  (Variables|Default|SetCanInteract false))

(fn ConstructionScript
  (|CallParentFunction))
