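; Temple vault door that opens when players match its nearby braziers to a randomized hint pattern.
; When the match is ready, the authority gathers braziers in a box, pairs puzzle and hint by Number Map, and records each expected state; a disabled hotfix opens the door instead.
; Check Door Solution compares lit states, then stops brazier interactions and opens the door.
; Timelines slide both leaves, with client-side dust, shake, audio and conversation-prompt sounds.
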(event Custom|Check%20Door%20Solution
  ;[0.02277 1 0.047] Check solution on torch lit / unlit event
  (bind _array_element (Utilities|Array|ForEachLoopwithBreak (Variables|Default|GetPuzzle%20Braziers))
    (:LoopBody
      (if (not (== (Variables|/ChronoEnviroItems/TemplePuzzles/PatternCheck/Actors/BP_Chrono_Pattern_Brazier.BP_Chrono_Pattern_Brazier_C|GetPuzzle%20Expected%20State _array_element) (Variables|/ChronoEnviroItems/TemplePuzzles/PatternCheck/Actors/BP_Chrono_Pattern_Brazier.BP_Chrono_Pattern_Brazier_C|Getb%20Is%20Lit _array_element)))
        (Variables|Default|SetPuzzle%20Solved false))
      (Variables|Default|SetPuzzle%20Solved true))
    (:Completed
      _))
  (if (Variables|Default|GetPuzzle%20Solved)
    (CallFunction|OpenVaultDoor)
    (bind _array_element_2 (Utilities|Array|ForEachLoopwithBreak (Variables|Default|GetPuzzle%20Braziers))
      (:LoopBody
        (Class|BPChronoPatternBrazier|StopInteractions _array_element_2)))))

(event EventBeginPlay
  ;[1 1 1] Init
  (Default|SetStartandEndPositionsforDoors)
  (Class|FortAsyncActionAuthorityMatchReady|MatchReadyAsync
    (:OnReady
      (if (ScalableFloat|GetValueasBool (Variables|Default|GetHotFixEnabled))
        ;[1 0.987 0.9767] Grab all nearby braziers, create pattern for puzzle, sync hint and puzzle braziers
        (Networking|SwitchHasAuthority
          (:Authority
            (bind _outactors (Class|SpecialEventHelpers|GetActorsinBox (Utilities|Array|MakeArray (Variables|Default|GetPattern%20Brazier%20Actor%20Class)) (Math|Box|MakeBox (- (Transformation|GetActorLocation) (ScalableFloat|GetValueatLevel (Variables|Default|GetMaxPuzzleHintBraziersNearRadius))) (+ (Transformation|GetActorLocation) (ScalableFloat|GetValueatLevel (Variables|Default|GetMaxPuzzleHintBraziersNearRadius))))))
            (Default|SetRandomPatternKeytoDoorfromBrazierArray _outactors)))
        (else
          (CallFunction|OpenVaultDoor))))))

(event Custom|Open
  (bind _audioloop (Variables|Default|GetAudioLoop))
  (bind _returnvalue (Transformation|GetWorldLocation _audioloop))
  ;[1 1 1] Open Door
  (Variables|Default|SetDoorOpened true)
  (Utilities|FlowControl|Sequence
    (:then_0
      (CallFunction|DoorAnimatingCosmetics) ; inc audio
    )
    (:then_1
      (bind (_direction _dooralpha) (Timeline|DoorOpenAnim PlayFromStart)
        (:Update
          (Niagara|SetNiagaraVariable(Float) (Variables|Default|GetNS_Dust) "DoorOpen" _dooralpha)
          (Transformation|SetWorldLocation (Variables|Default|GetSM_Ancients_VaultDoor_R) (Math|Vector|Lerp(Vector) (Variables|Default|GetStartLocation_R) (Variables|Default|GetEndLocation_R) _dooralpha) false false)
          (Transformation|SetWorldLocation (Variables|Default|GetSM_Ancients_VaultDoor_L) (Math|Vector|Lerp(Vector) (Variables|Default|GetStartLocation_L) (Variables|Default|GetEndLocation_L) _dooralpha) false false))
        (:EndShake
          (Audio|SpawnSoundatLocation (Variables|Sound|GetDoorStopSound) _returnvalue)
          (Audio|SpawnSoundatLocation (Variables|Sound|GetCamShakeSound) _returnvalue)
          (Camera|PlayWorldCameraShake (Variables|Default|GetShake) (Transformation|GetWorldLocation (Variables|Default|GetStaticMesh)) 2e+03 2e+03 0.75)
          (Components|DestroyComponent (Variables|Default|GetBox_Nav_Middle)) ; Remove Navigation Collision so AI can walk through the door now.
          (ForceFeedback|SpawnForceFeedbackatLocation (Variables|Default|GetForce%20Feedback%20Effect) (Transformation|GetWorldLocation (Variables|Default|GetStaticMesh))))))
    (:then_2
      (if (Variables|Default|GetDebugResetState)
        (Utilities|FlowControl|Delay 1e+01)
        (CallFunction|CloseVaultDoor)))))

(event Custom|OpenVaultDoor
  (bind _audioloop (Variables|Default|GetAudioLoop))
  (bind _returnvalue (Transformation|GetWorldLocation _audioloop))
  ;[1 1 1] Open Door
  (Variables|Default|SetDoorOpened true)
  (Utilities|FlowControl|Sequence
    (:then_0
      (CallFunction|DoorAnimatingCosmetics) ; inc audio
    )
    (:then_1
      (Timeline|DoorOpenAnim PlayFromStart))
    (:then_2
      (if (Variables|Default|GetDebugResetState)
        (Utilities|FlowControl|Delay 1e+01)
        (CallFunction|CloseVaultDoor)))))

(event Custom|CloseVaultDoor
  ;[1 1 1] Open Door
  (if (Variables|Default|GetDebugResetState)
    (Variables|Default|SetDoorOpened false)
    (else
      (Utilities|FlowControl|DoOnce false
        (:Completed
          (Variables|Default|SetDoorOpened true)))))
  (CallFunction|DoorAnimatingCosmetics)
  (bind (_direction _dooralpha) (Timeline|Timeline PlayFromStart)
    (:Update
      (Niagara|SetNiagaraVariable(Float) (Variables|Default|GetNS_Dust) "DoorOpen" _dooralpha)
      (Transformation|SetWorldLocation (Variables|Default|GetSM_Ancients_VaultDoor_R) (Math|Vector|Lerp(Vector) (Variables|Default|GetEndLocation_R) (Variables|Default|GetStartLocation_R) _dooralpha) false false)
      (Transformation|SetWorldLocation (Variables|Default|GetSM_Ancients_VaultDoor_L) (Math|Vector|Lerp(Vector) (Variables|Default|GetEndLocation_L) (Variables|Default|GetStartLocation_L) _dooralpha) false false))) ; door closing anim
)

(event Custom|DoorAnimatingCosmetics
  (bind _audioloop (Variables|Default|GetAudioLoop))
  (bind _returnvalue (Transformation|GetWorldLocation _audioloop))
  ;[1 1 1] FX and audio
  (Networking|SwitchHasAuthority
    (:Remote
      (Utilities|FlowControl|Sequence
        (:then_0
          (Camera|PlayWorldCameraShake (Variables|Default|GetShake) (Transformation|GetWorldLocation (Variables|Default|GetStaticMesh)) 2e+03 2e+03 0.75)
          (bind _returnvalue_2 (Niagara|SpawnSystemAttached "/ChronoEnviroItems/TemplePuzzles/RedLight/FX/NS_Chrono_Temple_DoorDust.NS_Chrono_Temple_DoorDust" (Variables|Default|GetStaticMesh) "None" (Transformation|GetWorldLocation (Variables|Default|GetStaticMesh)) "0, 0, 0" "KeepWorldPosition"))
          (Variables|Default|SetNS_Dust _returnvalue_2)
          (ForceFeedback|SpawnForceFeedbackatLocation (Variables|Default|GetForce%20Feedback%20Effect) (Transformation|GetWorldLocation (Variables|Default|GetStaticMesh))))
        (:then_1
          ;;[0.05153 0.09729 0.1969] Audio
          (Utilities|FlowControl|DoOnce false
            (:Completed
              (Audio|SpawnSoundatLocation (Variables|Sound|GetDoorOpenSound) (Transformation|GetWorldLocation (Variables|Default|GetAudioLoop)))
              (Audio|SpawnSoundatLocation (Variables|Sound|GetDoorOpenSound) _returnvalue)
              (Audio|SpawnSoundatLocation (Variables|Sound|GetDoorMovingSound) _returnvalue)
              (if (Variables|Default|GetDoorOpened)
                (bind _audioloop_2 (Variables|Default|GetAudioLoop)
                  (:then
                    (Audio|Components|Audio|FadeOut _audioloop_2 7.0 0.0 "Linear" :!class /Script/Engine.AudioComponent)
                    (Utilities|FlowControl|Delay 2.0)
                    (Utilities|FlowControl|Sequence
                      (:then_0
                        _)
                      (:then_1
                        _))
                    (Utilities|FlowControl|Delay 7.0)
                    (Utilities|FlowControl|DoOnce false
                      (:Completed
                        (Utilities|IsValid (Variables|Default|GetLocal%20Interacting%20Controller)
                          (:Is%20Valid
                            (Audio|SpawnSound2D (Variables|Sound|GetMusicStinger))
                            (Utilities|FlowControl|Delay 1.0))) ; Having this variable set means we're in the client that started interacting with this door.
                      )))
                  (:else
                    _)))
              (bind _audioloop_3 (Variables|Default|GetAudioLoop)
                (:then
                  (Audio|Components|Audio|FadeIn _audioloop_3 7.0 1.0 0.0 "Linear" :!class /Script/Engine.AudioComponent))
                (:else
                  _)))))))))

(event Custom|DestroyDoor
  (bind _self self)
  ;[1 1 1] Destroy
  (Variables|Default|SetDoorDestroyed true)
  (Networking|SwitchHasAuthority
    (:Authority
      (Utilities|IsValid _self
        (:Is%20Valid
          (Actor|DestroyActor _self))))
    (:Remote
      (Niagara|SpawnSystematLocation "/Game/Effects/Fort_Effects/World/Destruction/Base_Effects/NS_General_Deco_Destruction_Large_Converted.NS_General_Deco_Destruction_Large_Converted" (Transformation|GetWorldLocation (Variables|Default|GetDefaultSceneRoot)))
      (Rendering|SetActorHiddenInGame self true))))

(event OnReceivedLocalClientConversationMessage(NonPlayerConversationComponent)
  (params (LocalPlayerController :type Fort%20Player%20Controller%20Object%20Reference) (Message :type Client%20Conversation%20Message%20Payload%20Structure%20%28by%20ref%29))
  ;[0.05153 0.09729 0.1969] Audio For Text Prompts
  (Networking|SwitchHasAuthority
    (:Remote
      (bind _output_get (Variables|Default|SetLocal%20Interacting%20Controller LocalPlayerController))
      (Audio|SpawnSound2D (Variables|Sound|GetTextPromptOpenSound))
      (Ability|Tasks|WaitGameplayTagRemoveFromActor (Variables|/Script/Engine.Controller|GetPlayerState _output_get) "(TagName=\"Conversation.InteractingWithNPC\")" true)
      (CallFunction|OnClientFinishedConversation))))

(event Custom|OnClientFinishedConversation
  ;[0.05153 0.09729 0.1969] Audio For Text Prompts
  (Utilities|FlowControl|Sequence
    (:then_0
      (Audio|SpawnSound2D (Variables|Sound|GetTextPromptCloseSound)))
    (:then_1
      (Utilities|Time|ClearandInvalidateTimerbyHandle (Variables|Default|GetClientFinishesConversationTimer))
      (bind _returnvalue (Utilities|Time|SetTimerbyEvent (delegate Custom|CheckIfDoorOpenToUnsetController) 2.0))
      (Variables|Default|SetClientFinishesConversationTimer _returnvalue))))

(event Custom|CheckIfDoorOpenToUnsetController
  ;[0.05153 0.09729 0.1969] Audio For Text Prompts
  (if (not (Variables|Default|GetDoorOpened))
    (Variables|Default|SetLocal%20Interacting%20Controller 0)))

(fn BlueprintCanInteract
  (params (InteractingPawn :type Fort%20Pawn%20Object%20Reference) (InteractionBeingAttempted :type EInteractionBeingAttempted%20Enum :default "FirstInteraction") (InteractionType :type TInteractionType%20Enum :default "IT_NoInteraction"))
  (return true))

(fn BlueprintGetAllHighlightableComponents
  (return (Utilities|Array|MakeArray (Variables|Default|GetSM_Ancients_VaultDoor_R) (Variables|Default|GetSM_Ancients_VaultDoor_L)) true))

(fn BlueprintGetFailedInteractionString
  (params (InteractingPawn :type Fort%20Pawn%20Object%20Reference) (InteractionBeingAttempted :type EInteractionBeingAttempted%20Enum :default "FirstInteraction"))
  (return ""))

(fn BlueprintGetInteractionString
  (params (InteractingPawn :type Fort%20Pawn%20Object%20Reference) (InteractionBeingAttempted :type EInteractionBeingAttempted%20Enum :default "FirstInteraction"))
  (return (Variables|Default|GetInterationString)))

(fn BlueprintGetInteractionTime
  (params (InteractingPawn :type Fort%20Pawn%20Object%20Reference) (InteractionBeingAttempted :type EInteractionBeingAttempted%20Enum :default "FirstInteraction"))
  (return 0.0 true))

(fn OnRep_DoorDestroyed
  (if (Variables|Default|GetDoorDestroyed)
    (Rendering|SetVisibility (Variables|Default|GetStaticMesh) false false)
    (Collision|SetCollisionEnabled (Variables|Default|GetBlockAllCollision) "NoCollision")
    (else
      (Rendering|SetVisibility (Variables|Default|GetStaticMesh) true false))))

(fn OnRep_DoorOpened
  (if (Variables|Default|GetDoorOpened)
    (Collision|SetCollisionEnabled (Variables|Default|GetBlockAllCollision) "NoCollision")
    (Utilities|FlowControl|DoOnce false
      (:Completed
        (InteractionUI|AddHideUIIndicatorReason (Variables|Default|GetNonPlayerConversationComponent) (Variables|Default|GetHide%20Conversation%20UI%20Indicator%20Reason))
        (Variables|Interaction|SetbShowFirstInteractPrompt false)
        (Variables|Interaction|SetbAllowInteract false)))
    (else
      (Collision|SetCollisionEnabled (Variables|Default|GetBlockAllCollision) "QueryAndPhysics"))))

(fn SetRandomPatternKeyToDoorFromBrazierArray
  (params (Brazier%20Actor%20Array :type Array%20of%20Actor%20Object%20References))
  (bind _puzzle_hint_braziers (Variables|Default|GetPuzzle%20Hint%20Braziers))
  (for _array_element Brazier%20Actor%20Array
    ;[1 1 1] Grab hint braziers first, as they are child assets of braziers, and will pass cast
    (bind _asbp_chrono_pattern_brazier_solution_child (Utilities|Casting|CastToBP_Chrono_Pattern_Brazier_Solution_Child _array_element)
      (:then
        (Utilities|Array|AddUnique (Variables|Default|GetPuzzle%20Hint%20Braziers) _asbp_chrono_pattern_brazier_solution_child))
      (:CastFailed
        (bind _asbp_chrono_pattern_brazier (Utilities|Casting|CastToBP_Chrono_Pattern_Brazier _array_element))
        ;[1 1 1] Grab braziers
        (Utilities|Array|AddUnique (Variables|Default|GetPuzzle%20Braziers) _asbp_chrono_pattern_brazier)
        (Variables|/ChronoEnviroItems/TemplePuzzles/PatternCheck/Actors/BP_Chrono_Pattern_Brazier.BP_Chrono_Pattern_Brazier_C|SetDoor%20Ref self _asbp_chrono_pattern_brazier))))
  ;[1 1 1] Set correct starting state if we start with all unlit. Feedback has gone back and forth, so hotfixing starting lit/unlit
  (if (not (ScalableFloat|GetValueasBool (Variables|Default|GetStartBraziersLitHotfix)))
    (for _array_element_2 (Variables|Default|GetPuzzle%20Braziers)
      (Class|BPChronoPatternBrazier|ExtinguishBrazier _array_element_2)
      (Class|BPChronoPatternBrazier|StopSmokeonStartUnlit _array_element_2))
    (bind _array_element_3 (Utilities|Array|ForEachLoopwithBreak (Variables|Default|GetPuzzle%20Hint%20Braziers))
      (:LoopBody
        (Class|BPChronoPatternBrazier|ExtinguishBrazier _array_element_3))
      (:Completed
        _)))
  ;[1 1 1] Match puzzle and hint by Number Map association
  (for _array_element_4 (Variables|Default|GetPuzzle%20Braziers)
    (bind _array_element_5 (Utilities|Array|ForEachLoopwithBreak (Variables|Default|GetPuzzle%20Hint%20Braziers))
      (:LoopBody
        (if (== (Variables|/ChronoEnviroItems/TemplePuzzles/PatternCheck/Actors/BP_Chrono_Pattern_Brazier.BP_Chrono_Pattern_Brazier_C|GetNumber%20Map _array_element_5) (Variables|/ChronoEnviroItems/TemplePuzzles/PatternCheck/Actors/BP_Chrono_Pattern_Brazier.BP_Chrono_Pattern_Brazier_C|GetNumber%20Map _array_element_4))
          (Variables|/ChronoEnviroItems/TemplePuzzles/PatternCheck/Actors/BP_Chrono_Pattern_Brazier.BP_Chrono_Pattern_Brazier_C|SetPuzzle%20Brazier%20Ref _array_element_4 _array_element_5)
          (Variables|/ChronoEnviroItems/TemplePuzzles/PatternCheck/Actors/BP_Chrono_Pattern_Brazier.BP_Chrono_Pattern_Brazier_C|SetHint%20Brazier%20Ref _array_element_5 _array_element_4)))))
  ;[1 1 1] Create a random pattern by extinguishing some of the Hint braziers, then map that state to be the expected state to reference on interaction and solution check
  (for _array_element_6 (Variables|Default|GetPuzzle%20Hint%20Braziers)
    (if (> (Math|Random|RandomIntegerinRange 0 100) 50)
      (if (ScalableFloat|GetValueasBool (Variables|Default|GetStartBraziersLitHotfix))
        (Class|BPChronoPatternBrazier|ExtinguishBrazier _array_element_6)
        (else
          (Class|BPChronoPatternBrazier|IgniteBrazier _array_element_6)))))
  ;[1 1 1] Ensure the pattern is not all lit / unlit ( actually all unlit isn't really an issue... )
  (for _array_element_7 (Variables|Default|GetPuzzle%20Hint%20Braziers)
    (if (ScalableFloat|GetValueasBool (Variables|Default|GetStartBraziersLitHotfix))
      (if (Variables|/ChronoEnviroItems/TemplePuzzles/PatternCheck/Actors/BP_Chrono_Pattern_Brazier.BP_Chrono_Pattern_Brazier_C|Getb%20Is%20Lit _array_element_7)
        _
        (else
          _))
      (elif (Variables|/ChronoEnviroItems/TemplePuzzles/PatternCheck/Actors/BP_Chrono_Pattern_Brazier.BP_Chrono_Pattern_Brazier_C|Getb%20Is%20Lit _array_element_7)
        _))
    (Variables|Default|SetSetPatternOrForcePatternCondition true))
  (if (not (Variables|Default|GetSetPatternOrForcePatternCondition))
    ;;[0.1674 1 0.1911] hardcode a failsafe to extinguish first and last ( we could always go back through the randomization a few times before we do this, but this shoudl be fine )
    (for _array_element_8 _puzzle_hint_braziers
      (if (or (Utilities|Array|IsValidIndex _puzzle_hint_braziers) (Utilities|Array|IsValidIndex _puzzle_hint_braziers (Utilities|Array|LastIndex _puzzle_hint_braziers)))
        (if (ScalableFloat|GetValueasBool (Variables|Default|GetStartBraziersLitHotfix))
          (Class|BPChronoPatternBrazier|ExtinguishBrazier _array_element_8)
          (else
            (Class|BPChronoPatternBrazier|IgniteBrazier _array_element_8))))))
  ;[0 0.9305 1] Sync Expected solution state
  (for _array_element_9 (Variables|Default|GetPuzzle%20Braziers)
    (Variables|/ChronoEnviroItems/TemplePuzzles/PatternCheck/Actors/BP_Chrono_Pattern_Brazier.BP_Chrono_Pattern_Brazier_C|SetPuzzle%20Expected%20State (Variables|/ChronoEnviroItems/TemplePuzzles/PatternCheck/Actors/BP_Chrono_Pattern_Brazier.BP_Chrono_Pattern_Brazier_C|Getb%20Is%20Lit (Variables|/ChronoEnviroItems/TemplePuzzles/PatternCheck/Actors/BP_Chrono_Pattern_Brazier.BP_Chrono_Pattern_Brazier_C|GetHint%20Brazier%20Ref _array_element_9)) _array_element_9)))

(fn SetStartAndEndPositionsForDoors
  (bind _sm_ancients_vaultdoor_l (Variables|Default|GetSM_Ancients_VaultDoor_L))
  (bind _returnvalue (Transformation|GetWorldLocation _sm_ancients_vaultdoor_l))
  (bind _sm_ancients_vaultdoor_r (Variables|Default|GetSM_Ancients_VaultDoor_R))
  (bind _returnvalue_2 (Transformation|GetWorldLocation _sm_ancients_vaultdoor_r))
  ;[1 1 1] Store off base 'start' position. Right door needs to go 'backwards' negative units
  (Variables|Default|SetStartLocation_R _returnvalue_2)
  (Variables|Default|SetEndLocation_R (+ _returnvalue_2 (* (Transformation|GetForwardVector _sm_ancients_vaultdoor_r) (Math|Conversions|ToVector(Float) (Variables|Default|GetNegativeDoorExtensionLength)))))
  ;[1 1 1] Store off base 'start' position. Left door needs to go 'forwards' positive units
  (Variables|Default|SetStartLocation_L _returnvalue)
  (Variables|Default|SetEndLocation_L (+ _returnvalue (* (Transformation|GetForwardVector _sm_ancients_vaultdoor_l) (Math|Conversions|ToVector(Float) (Variables|Default|GetDoorExtensionLength)))))
  (return))

(fn ConstructionScript)
